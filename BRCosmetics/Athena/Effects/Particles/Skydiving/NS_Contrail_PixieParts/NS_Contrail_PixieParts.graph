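# Niagara skydiving contrail cosmetic "PixieParts", drawing a glowing trail around the player's skeletal mesh.
# Swirls, TrailCenter and two visibility-tagged Sprites groups sample one bone, while SymbolPlate follows another with layered meshes and a light.
# Sparks spawn from SymbolPlate particles through a particle attribute reader.
# User.ParaGlideLeanAlpha and System.Dive blend rates, motion and scale, System.TOD tints colors, and front-end modifiers retune values for menus.
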NiagaraSystem /BRCosmetics/Athena/Effects/Particles/Skydiving/NS_Contrail_PixieParts/NS_Contrail_PixieParts
dependencies {
 usedDataInterfaces [
  &NiagaraDataInterfaceArrayInt32
  &NiagaraDataInterfaceCurve
  &NiagaraDataInterfaceVector2DCurve
  &NiagaraDataInterfaceArrayFloat
  &NiagaraDataInterfaceParticleRead
  &NiagaraDataInterfaceSkeletalMesh
 ]
 usedDynamicInputs [
  &SelectFloatFromArray
  &SetBoolByFloatComparison
  /Niagara/DynamicInputs/Math/Clamp_Vector
  &TransformVector
 ]
 usedModules [
  &SystemState
  &self:SystemSpawnScript.SystemScriptSource.SystemScriptGraph.NiagaraNodeAssignment_4.SetVariables_#1
  &self:SystemSpawnScript.SystemScriptSource.SystemScriptGraph.NiagaraNodeAssignment_0.SetVariables_#2
  &FN_Integrator
  &self:SystemSpawnScript.SystemScriptSource.SystemScriptGraph.NiagaraNodeAssignment_1.SetVariables_#3
  &self:SystemSpawnScript.SystemScriptSource.SystemScriptGraph.NiagaraNodeAssignment_2.SetVariables_#4
  &self:SystemSpawnScript.SystemScriptSource.SystemScriptGraph.NiagaraNodeAssignment_3.SetVariables_#5
  &self:SystemSpawnScript.SystemScriptSource.SystemScriptGraph.NiagaraNodeAssignment_5.SetVariables_#6
 ]
}
emitters {
 Sparks {
  emitterUpdateScript [
   &EmitterState()
   &SpawnParticlesFromOtherEmitter(
    Attribute Reader=DataInterface() {propertyValues {EmitterBinding {bindingMode=Other, emitterName=SymbolPlate}}}
    Spawn Rate=&Lerp_Float(A=20, B=50, Alpha=$User.ParaGlideLeanAlpha)
    Use Maximum Age=false
    Use Minimum Age=false
   )
  ]
  particleSpawnScript [
   &InitializeParticle(
    Color Maximum=<rgba 0.0632 10 6.5845 1>
    Color Minimum=<rgba 0.0374 0.7671 2 1>
    Color Mode=::NewEnumerator2
    Lifetime=&Lerp_Float(A=1, B=0.25, Alpha=$User.ParaGlideLeanAlpha)
    Sprite Size Max=<2 3>
    Sprite Size Min=<1 2>
    Sprite Size Mode=::NewEnumerator2
    UsePositionOffset=false
    Write Material Random=true
   )
   &SampleParticlesFromOtherEmitter()
   &ShapeLocation(
    Large Radius=20
    MaskSpawnGroup=false
    Offset=&Lerp_Vector(A=<0 0 0>, B=<0 0 7>, Alpha=$User.ParaGlideLeanAlpha)
    Offset Mode=::NewEnumerator0
    Override Seed=false
    Shape Origin=$Output.SampleParticlesFromOtherEmitter.SampledPosition
    Shape Primitive=::NewEnumerator3
    UseOwnerScale=true
   )
   &AddVelocity(Velocity=&Lerp_Vector(A=<0 0 10>, B=<0 0 400>, Alpha=$User.ParaGlideLeanAlpha))
   &AddVelocity(
    Module.Use Falloff Distance=false
    Origin Offset=&Lerp_Vector(A=<0 0 0>, B=<0 0 7>, Alpha=$User.ParaGlideLeanAlpha)
    Override Seed=false
    Velocity Mode=::NewEnumerator1
    Velocity Origin=$Output.SampleParticlesFromOtherEmitter.SampledPosition
    Velocity Speed=&Lerp_Float(A=122.0596, B=250, Alpha=$User.ParaGlideLeanAlpha)
   ) {module=AddVelocity001}
  ]
  particleUpdateScript [
   &ParticleState()
   &GravityForce(Gravity=&Multiply_Vector(
     A=&RandomRangeVector(Minimum=<-100 -100 300>, Maximum=<100 100 500>)
     B=&Lerp_Vector(A=<1 1 1>, B=<1 1 5>, Alpha=$User.ParaGlideLeanAlpha)
    ))
   &CurlNoiseForce(Bias Noise=false, Noise Strength=284.9621, Pan Noise=false)
   &SolveForcesAndVelocity(Clamp Velocity=false, Limit Acceleration=false)
   &ScaleColor(
    Scale Alpha=&FloatFromCurve(FloatCurve=DataInterface() {propertyValues {Curve {keys [{interpMode=RCIM_Cubic, tangentMode=RCTM_User, time=0, value=0}, <at 0.0925 1>, <at 1 0>]}}})
    Scale RGB=&MakeVectorFromLinearColorRGB(Color=$System.TOD)
    ScaleA=true
    ScaleRGB=true
   )
  ]
  rendererClasses [&NiagaraSpriteRendererProperties]
  renderers [{rendererClass=&NiagaraSpriteRendererProperties}]
 }
 Sprites {
  emitterUpdateScript [
   &EmitterState()
   &SpawnRate(SpawnRate=&Lerp_Float(A=3, B=6, Alpha=$User.ParaGlideLeanAlpha), Use Spawn Probability=false)
   &SpawnRate(SpawnGroup=1, SpawnRate=&Lerp_Float(A=3, B=6, Alpha=$User.ParaGlideLeanAlpha), Use Spawn Probability=false) {module=SpawnRate001}
  ]
  particleSpawnScript [
   &InitializeParticle(
    Color=$System.TOD
    Lifetime=&Lerp_Float(A=1, B=0.506, Alpha=$User.ParaGlideLeanAlpha)
    Sprite Size Mode=::NewEnumerator4
    UsePositionOffset=false
    Write Material Random=true
   )
   &SkeletalMeshLocation(
    Bone / Socket Index=7
    Bone Sampling Mode=::NewEnumerator4
    Inherit Velocity=false
    Interpolate Mesh Sampling=false
    MaskSpawnGroup=false
    Skeletal Mesh=$System.SkeletalMesh
   )
   &ShapeLocation(
    Large Radius=35.2972
    MaskSpawnGroup=false
    Offset=<0 0 40>
    Offset Mode=::NewEnumerator0
    Override Seed=false
    Shape Primitive=::NewEnumerator3
    UseOwnerScale=true
   )
   &AddVelocity(Velocity=&Lerp_Vector(A=<0 0 10>, B=<0 0 55>, Alpha=$User.ParaGlideLeanAlpha))
   &AddVelocity(Module.Use Falloff Distance=false, Origin Offset=<0 0 55>, Override Seed=false, Velocity Mode=::NewEnumerator1, Velocity Speed=59.7188) {module=AddVelocity001}
   /SetVariables(
    Particles.ColorDudes:LinearColor=&MultiplyLinearColor(
     Linear Color=<rgba 1.1 1.1 1.1 1.5>
     Scale Factor=&MakeVector4_FromLinearColor(LinearColor=/Game/Effects/Niagara/DynamicInputs/TimeOfDay/DaySequenceTimeOfDayColor(Linear Color:LinearColor=<rgba 1 1 1 1>, Time Of Day Influence:NiagaraFloat=1))
    )
    Particles.DudesIndex:NiagaraFloat=&MakeFloatFromInt(Int=&RandomRangeInt(Minimum=0, Maximum=4))
    Particles.SpriteSizeDudes:Vector2f=<30 20>
    Particles.SubImageIndex=&MakeFloatFromInt(Int=&RandomRangeInt(Minimum=0, Maximum=3))
    Particles.VisibilityTag=&SelectIntFromArray(Direct Array Index=&SpawnGroup(), Int Selection Array=DataInterface() {propertyValues {IntData [0,1]}})
   ) {isSetParameters=true}
  ]
  particleUpdateScript [
   &ParticleState()
   &SpriteRotationRate(Rotation Rate=&RandomRangeFloat(Minimum=-555, Maximum=555))
   &ScaleSpriteSize(Uniform Curve Sprite Scale=DataInterface() {propertyValues {Curve {
       defaultValue=0
       keys [
        {arriveTangent=15.2458, interpMode=RCIM_Cubic, leaveTangent=15.2458, tangentMode=RCTM_User, time=0, value=0}
        {arriveTangent=1.2688, interpMode=RCIM_Cubic, leaveTangent=1.2688, time=0.1615, value=1.8183}
        {arriveTangent=-1.4029, interpMode=RCIM_Cubic, leaveTangent=-1.4029, time=0.4668, value=0.5922}
        {interpMode=RCIM_Cubic, tangentMode=RCTM_User, time=0.7725, value=0.9612}
        {tangentMode=RCTM_User, time=1, value=0}
       ]
      }}})
   /SetVariables(Particles.SpriteSizeDudes:Vector2f=&Multiply_Vector2D(
     A=$Particles.Initial.SpriteSizeDudes
     B=&Vector2DFromCurve(
      Vector2Curve=DataInterface() {propertyValues {
        XCurve {keys [
          {interpMode=RCIM_Cubic, leaveTangent=14.5412, tangentMode=RCTM_Break, time=0, value=0}
          {interpMode=RCIM_Cubic, tangentMode=RCTM_Break, time=0.1789, value=1}
          {arriveTangent=-2.672, interpMode=RCIM_Cubic, leaveTangent=2.4441, tangentMode=RCTM_Break, time=0.482, value=0.7549}
          {arriveTangent=-6.2448, interpMode=RCIM_Cubic, tangentMode=RCTM_Break, time=0.7471, value=0.5647}
          {arriveTangent=-7.5036, interpMode=RCIM_Cubic, tangentMode=RCTM_Break, time=1, value=0}
         ]}
        YCurve {keys [
          {interpMode=RCIM_Cubic, leaveTangent=17.7002, tangentMode=RCTM_Break, time=0, value=0}
          {interpMode=RCIM_Cubic, tangentMode=RCTM_Break, time=0.1727, value=1.1636}
          {arriveTangent=-3.9162, interpMode=RCIM_Cubic, leaveTangent=0.5318, tangentMode=RCTM_Break, time=0.482, value=0.7549}
          {arriveTangent=-2.4024, interpMode=RCIM_Cubic, tangentMode=RCTM_Break, time=0.7471, value=0.5647}
          {arriveTangent=-7.5036, interpMode=RCIM_Cubic, tangentMode=RCTM_Break, time=1, value=0}
         ]}
       }}
      ScaleCurve=<0.6 1>
     )
    )) {isSetParameters=true}
   &GravityForce(Gravity=&Multiply_Vector(
     A=&RandomRangeVector(Minimum=<-100 -100 300>, Maximum=<100 100 500>)
     B=&Lerp_Vector(A=<1 1 1>, B=<1 1 5>, Alpha=$User.ParaGlideLeanAlpha)
    ))
   &CurlNoiseForce(Bias Noise=false, Noise Strength=284.9621, Pan Noise=false)
   &SolveForcesAndVelocity(Clamp Velocity=false, Limit Acceleration=false)
  ]
  rendererClasses [&NiagaraSpriteRendererProperties]
  renderers [{rendererClass=&NiagaraSpriteRendererProperties}]
 }
 Swirls {
  emitterUpdateScript [
   &EmitterState()
   &SpawnRate(
    SpawnGroup=1
    SpawnRate=&Lerp_Float(A=6, B=&FN_FrontEndFloatModifier(Gameplay=9, Front End=6), Alpha=$System.Dive)
    Use Spawn Probability=false
   )
  ]
  particleSpawnScript [
   &InitializeParticle(
    Color Maximum=<rgba 0 1.3632 2 0.7826>
    Color Minimum=<rgba 0.0004 0 4 0.6783>
    Color Mode=::NewEnumerator2
    Lifetime=&Lerp_Float(
     A=&RandomRangeFloat(Minimum=1, Maximum=3)
     B=&FN_FrontEndFloatModifier(
      Gameplay=&RandomRangeFloat(Minimum=0.4, Maximum=0.5)
      Front End=&RandomRangeFloat(Minimum=1, Maximum=1.5)
     )
     Alpha=$System.Dive
    )
    Mesh Scale=&RandomRangeVector(Minimum=<3 1.5 1.5>, Maximum=<4 2 2>)
    Mesh Scale Mode=::NewEnumerator4
    UsePositionOffset=true
    Write Material Random=true
   )
   &CameraOffset(Camera Offset Amount=3)
   &InitialMeshOrientation(
    Mesh Axis To Orient=<0 0 -1>
    Mesh Orientation Mode=::NewEnumerator2
    Orientation Vector=<0 0 1>
    Rotation=&RandomRangeVector(Minimum=<0 0 0>, Maximum=<0 0 1>)
    Rotation Coordinate Space=::NewEnumerator2
    Use Rotation Vector=true
   ) {module=InitialMeshOrientation001}
  ]
  particleUpdateScript [
   &ParticleState()
   &InterpolateOverTime(Rate Of Change=50, Target Vector Value:NiagaraPosition=$Particles.Position, Variable Type=::NewEnumerator2)
   &DynamicMaterialParameters(
    Index 0 Param 1=&RandomRangeFloat(Minimum=1, Maximum=2)
    Index 0 Param 2=&Multiply_Float(
     A=&RandomRangeFloat(Minimum=4, Maximum=6)
     B=&Lerp_Float(A=1, B=&FN_FrontEndFloatModifier(Gameplay=2.5, Front End=1.5), Alpha=$System.Dive)
    )
    Index 0 Param 3=&Multiply_Float(
     A=&RandomRangeFloat(Minimum=5, Maximum=12)
     B=&Lerp_Float(A=1, B=&FN_FrontEndFloatModifier(Gameplay=2.5, Front End=1.5), Alpha=$System.Dive)
    )
    Index 0 Param 4=&RandomRangeFloat(Minimum=1, Maximum=1.6)
    Index 1 Param 4=0.7
    Index 1 Vector 1:Vector3f=$Particles.InterpolateOverTime.InterpolateOverTime_V3_.Moving Average
    Index 2 Param 1=&FloatFromCurve(FloatCurve=DataInterface() {propertyValues {Curve {keys [<at 0 1 RCIM_Cubic>,<at 0.4739 1 RCIM_Cubic>,<at 1 1>]}}})
    Index 2 Param 2=&FloatFromCurve(FloatCurve=DataInterface() {propertyValues {Curve {keys [<at 0 -1>,<at 1 1>]}}}, Scale Curve=-1)
    Index 2 Param 3=&RandomRangeFloat(Minimum=10, Maximum=33)
    Index 2 Param 4=1
    Module.Index0Float0Write=true
    Module.Index0Float1Write=true
    Module.Index0Float2Write=true
    Module.Index0Float3Write=true
    Module.Index1Float3Write=true
    Module.Index1VectorWrite=true
    Module.Index2Float0Write=true
    Module.Index2Float1Write=true
    Module.Index2Float2Write=true
    Module.Index2Float3Write=false
    Write Parameter Index 1=::NewEnumerator3
    Write Parameter Index 2=::NewEnumerator1
   )
   &SkeletalMeshLocation(
    Bone / Socket Index=7
    Bone Sampling Mode=::NewEnumerator4
    MaskSpawnGroup=false
    Orientation Sampling=::NewEnumerator1
    Sampled Position Offset=<0 25 0>
    Skeletal Mesh=$System.SkeletalMesh
   )
   /SetVariables(Particles.Position=$Output.SkeletalMeshLocation.SampledPositionWithoutOffset) {isSetParameters=true}
   &ScaleMeshSize(Scale Factor=&VectorFromFloat(Value=&FloatFromCurve(
      FloatCurve=DataInterface() {propertyValues {Curve {keys [{arriveTangent=1.0186, interpMode=RCIM_Cubic, leaveTangent=1.0186, tangentMode=RCTM_User, time=0, value=0}, <at 1 1 RCIM_Cubic>]}}}
      CurveIndex=$Emitter.Age
      Scale Curve=2
     )))
   &ScaleColor(Color Scale:LinearColor=$System.TOD, Scale Mode=3)
  ]
  rendererClasses [&NiagaraMeshRendererProperties]
  renderers [{rendererClass=&NiagaraMeshRendererProperties, rendererIndex=0}]
 }
 SymbolPlate {
  emitterUpdateScript [
   &EmitterState(
    Life Cycle Mode=::NewEnumerator1
    Loop Behavior=::NewEnumerator1
    Loop Duration Mode=::NewEnumerator1
    Reset Age On Awaken=true
    Scalability Mode=::NewEnumerator1
    UseLoopDelay=false
   )
   &SpawnBurst_Instantaneous(Use Spawn Probability=false, UseLoopCountLimit=false)
  ]
  particleSpawnScript [
   &InitializeParticle(
    Color=$System.TOD
    Lifetime=2
    Mesh Scale Mode=::NewEnumerator3
    Mesh Uniform Scale=1.25
    UsePositionOffset=false
    Write Material Random=true
   ) {module=InitializeParticle001}
   &InitialMeshOrientation(
    Mesh Axis To Orient=<0 1 0>
    Mesh Orientation Mode=::NewEnumerator2
    Orientation Vector=<0 0 1>
    Rotation=&FN_FrontEndVector3Modifier(Gameplay=<0 0 -0.5>, Front End=<-0.05 0 -0.5>)
    Use Rotation Vector=true
   )
  ]
  particleUpdateScript [
   &ParticleState(Kill Particles When Lifetime Has Elapsed=false, Let Infinitely Lived Particles Die When Emitter Deactivates=true)
   &SkeletalMeshLocation(
    Bone / Socket Index=8
    Bone Sampling Mode=::NewEnumerator4
    MaskSpawnGroup=false
    Orientation Sampling=::NewEnumerator1
    Sampled Orientation Coordinate Space=::Local
    Sampled Position Offset=&FN_FrontEndVector3Modifier(
     Gameplay=&Lerp_Vector(A=<0 -5 -70>, B=<0 0 -120>, Alpha=$User.ParaGlideLeanAlpha)
     Front End=&Lerp_Vector(A=<0 0 -50>, B=<0 0 -70>, Alpha=$User.ParaGlideLeanAlpha)
    )
    Sampled Position Offset Coordinate Space:ENiagaraSkelMeshTransforms=::NewEnumerator1
    Skeletal Mesh=$System.SkeletalMesh
   )
   &UpdateMeshOrientation(
    Facing Direction=&Multiply_Vector(A=Linked() {linkedVariable {name=Particles.SkeletalMeshLocation.SampledPosition, type=NiagaraPosition}}, B=<-1 1 1>)
    Facing Interpolation=true
    Facing Interpolation Rate:NiagaraFloat=&FN_FrontEndFloatModifier(Gameplay=4, Front End=3)
    Orientation Method=::NewEnumerator1
    Override Quat=false
    TransformFacing=false
    TransformUp=false
    Up Direction:Vector3f=<1 0 0>
   )
   &ScaleMeshSize(Scale Factor=&VectorFromFloat(Value=&Lerp_Float(A=1, B=&FN_FrontEndFloatModifier(Gameplay=1.3, Front End=1.2), Alpha=$User.ParaGlideLeanAlpha)))
   &SolveForcesAndVelocity(Clamp Velocity=false, Limit Acceleration=false)
   &ScaleColor(
    Scale Alpha=&FloatFromCurve(FloatCurve=DataInterface() {propertyValues {Curve {keys [
         {arriveTangent=4.0516, interpMode=RCIM_Cubic, leaveTangent=4.0516, tangentMode=RCTM_User, time=0, value=0}
         {arriveTangent=0.0961, interpMode=RCIM_Cubic, leaveTangent=0.0961, tangentMode=RCTM_User, time=0.25, value=1}
         {arriveTangent=0.2272, interpMode=RCIM_Cubic, leaveTangent=0.2272, tangentMode=RCTM_User, time=1, value=1}
        ]}}})
    Scale RGB=&Lerp_Vector(
     A=<1 1 1.5>
     B=&FN_FrontEndVector3Modifier(Gameplay=<1 1 3>, Front End=<1 1 1.5>)
     Alpha=$User.ParaGlideLeanAlpha
    )
    ScaleA=true
    ScaleRGB=true
   )
   &DynamicMaterialParameters(
    Index 0 Param 1=1
    Index 0 Param 2=1
    Index 0 Param 3=&FloatFromCurve(FloatCurve=DataInterface() {propertyValues {Curve {keys [{arriveTangent=1.3962, interpMode=RCIM_Cubic, leaveTangent=1.3962, tangentMode=RCTM_User, time=0.0032, value=0.136}, <at 1 1>]}}})
    Index 0 Param 4=&FloatFromCurve(
     FloatCurve=DataInterface() {propertyValues {Curve {keys [{arriveTangent=-1.0048, interpMode=RCIM_Cubic, leaveTangent=-1.0048, tangentMode=RCTM_User, time=-0.0046, value=1.0113}, <at 0.7 0.3>]}}}
     Scale Curve=0.5
    )
    Module.Param0WriteEnabled=false
    Module.Param1WriteEnabled=true
    Module.Param2WriteEnabled=true
    Module.Param3WriteEnabled=true
   )
   &CameraOffset(Camera Offset Amount=-25)
   /SetVariables(
    Particles.SemiSphere_Color:LinearColor=&MakeLinearColorFromVectorAndFloat(
     "Vector (RGB)"=&MakeVectorFromLinearColorRGB(Color=&Multiply_LinearColorByFloat(
       Linear Color=$System.TOD
       Float=&FN_FrontEndFloatModifier(
        Gameplay=&Lerp_Float(A=1, B=1.5, Alpha=$User.ParaGlideLeanAlpha)
        Front End=&Lerp_Float(A=1, B=1.2, Alpha=$User.ParaGlideLeanAlpha)
       )
      ))
     "Float (Alpha)"=&FloatFromCurve(FloatCurve=DataInterface() {propertyValues {Curve {keys [
          {arriveTangent=4.0516, interpMode=RCIM_Cubic, leaveTangent=4.0516, tangentMode=RCTM_User, time=0, value=0}
          {arriveTangent=0.0961, interpMode=RCIM_Cubic, leaveTangent=0.0961, tangentMode=RCTM_User, time=0.25, value=1}
          {arriveTangent=0.2272, interpMode=RCIM_Cubic, leaveTangent=0.2272, tangentMode=RCTM_User, time=1, value=1}
         ]}}})
    )
    Particles.SemiSphere_Scale:Vector3f=&VectorFromFloat(Value=&Multiply_Float(
      A=1.1
      B=&Lerp_Float(A=1, B=&FN_FrontEndFloatModifier(Gameplay=1.2, Front End=1.2), Alpha=$User.ParaGlideLeanAlpha)
     ))
   ) {isSetParameters=true}
   &FN_Light_Attributes(
    Create Exponent Attribute=true
    Create Position Attribute=true
    Create Radius Attribute=true
    Exponent=2
    Position=&AddVectorToPosition(Position=$Particles.Position, Vector=<0 0 20>)
    Radius=60
   )
  ]
  rendererClasses [&NiagaraMeshRendererProperties, &NiagaraLightRendererProperties]
  renderers [
   {rendererClass=&NiagaraMeshRendererProperties, rendererIndex=0}
   {rendererClass=&NiagaraMeshRendererProperties, rendererIndex=1}
   {rendererClass=&NiagaraLightRendererProperties, rendererIndex=2}
  ]
 }
 TrailCenter {
  emitterUpdateScript [
   &EmitterState(
    Inactive Response=::NewEnumerator1
    Life Cycle Mode=::NewEnumerator1
    Loop Behavior=::NewEnumerator1
    Loop Duration Mode=::NewEnumerator1
    UseLoopDelay=false
   )
   &SpawnBurst_Instantaneous(Spawn Count=1, Use Spawn Probability=false)
  ]
  particleSpawnScript [
   &InitializeParticle(
    Color=<rgba 0.0402 7 5.4691 0.2087>
    Mesh Scale=&MakeCustomVectorFromBool(Bool=$User.bIsFrontEnd, True Vector=<2.5 2.5 2>, False Vector=<1 1 1>)
    Mesh Scale Mode=::NewEnumerator4
    Sprite Size Mode=::NewEnumerator3
    Uniform Sprite Size=333
    UsePositionOffset=false
    Write Material Random=true
   )
   &DynamicMaterialParameters(
    Index 0 Param 1=&RandomRangeFloat(Minimum=0.5, Maximum=0.8)
    Index 0 Param 2=&RandomRangeFloat(Minimum=0, Maximum=10)
    Index 0 Param 3=30
    Index 0 Param 4=1
    Module.Param0WriteEnabled=true
    Module.Param1WriteEnabled=true
    Module.Param2WriteEnabled=false
    Module.Param3WriteEnabled=true
   )
   /SetVariables(Particles.RotatorSpeedRandom:NiagaraFloat=&RandomRangeFloat(Minimum=0.3, Maximum=-0.3)) {isSetParameters=true}
   &CameraOffset(Camera Offset Amount=5)
   &InitialMeshOrientation(Orientation Axis=<0 0 1>, Orientation Vector=<0 0 1>, Use Orientation Vector=true, Use Rotation Vector=false)
  ]
  particleUpdateScript [
   &ParticleState(Kill Particles When Lifetime Has Elapsed=false, Let Infinitely Lived Particles Die When Emitter Deactivates=true)
   &InterpolateOverTime(Target Vector Value:NiagaraPosition=$Particles.Position, Variable Type=::NewEnumerator2)
   /SetVariables(Particles.InterpolateDynamic:Vector4f=&MakeFloat4FromVectorAndFloat("Vector (RGB)"=$Output.InterpolateOverTime.Moving Average, "Float (Alpha)"=1)) {isSetParameters=true}
   &SkeletalMeshLocation(
    Bone / Socket Index=7
    Bone Sampling Mode=::NewEnumerator4
    MaskSpawnGroup=false
    Orientation Sampling=::NewEnumerator1
    Sampled Position Offset=<0 -6 0>
    Skeletal Mesh=$System.SkeletalMesh
   )
   &SolveForcesAndVelocity(Clamp Velocity=false, Limit Acceleration=false)
   &DynamicMaterialParameters(
    Index 0 Param 3=&Multiply_Float(A=45, B=&Lerp_Float(A=0.5, B=1, Alpha=$System.Dive))
    Index 2 Param 1=&Lerp_Float(A=0, B=1.5, Alpha=&Ceil_Float(Float=$System.Dive))
    Index 2 Param 2=&Lerp_Float(A=0, B=0.5, Alpha=&Ceil_Float(Float=$System.Dive))
    Module.Param0WriteEnabled=false
    Module.Param0WriteEnabled2=true
    Module.Param1WriteEnabled=false
    Module.Param1WriteEnabled2=true
    Module.Param2WriteEnabled=true
    Module.Param2WriteEnabled2=false
    Module.Param3WriteEnabled=false
    Module.Param3WriteEnabled2=false
    Write Parameter Index 2=true
   ) {module=DynamicMaterialParameters001}
   &ScaleMeshSize(Scale Factor=&MakeCustomVectorFromBool(
     Bool=$User.bIsFrontEndPreview
     True Vector=&Lerp_Vector(A=<1 1 0.6>, B=<0.85 0.85 0.8>, Alpha=$System.Dive)
     False Vector=&Lerp_Vector(A=<1.5 1.5 0.9>, B=<0.85 0.85 0.8>, Alpha=$System.Dive)
    ))
   &ScaleMeshSize(Scale Factor=<1.1 1.1 1>) {module=ScaleMeshSize001}
   &ScaleColor(
    Scale Alpha=&FloatFromCurve(
     FloatCurve=DataInterface() {propertyValues {Curve {defaultValue=1, keys [{tangentMode=RCTM_User, time=0, value=0}, {tangentMode=RCTM_User, time=0.3, value=1}]}}}
     CurveIndex=$Engine.System.Age
    )
    ScaleA=true
    ScaleRGB=true
   )
   &CameraOffset(Camera Offset Amount=5) {module=CameraOffset001}
   &FN_Sprite_Attributes(
    Camera Offset=18.5
    Color=&MultiplyLinearColor(
     Linear Color=<rgba 0.2348 0.4782 1 0.3>
     Scale Factor=&MakeVector4_FromLinearColor(LinearColor=$System.TOD)
    )
    Create Camera Offset Attribute=true
    Create Color Attribute=true
    Create Dynamic Material Parameter Attributes=true
    Create Position Attribute=true
    Create SpriteSize Attribute=true
    Index 0 Param 1=3
    Index 0 Param 2=1
    Index 0 Param 3=1
    Index 0 Param 4=1
    Module Output:ENiagara_AttributeSamplingApplyOutput=::NewEnumerator1
    Module.Index0Float0Write=true
    Module.Index0Float1Write=true
    Module.Index0Float2Write=true
    Module.Index0Float3Write=true
    Position:NiagaraPosition=&AddVectorToPosition(Position=$Output.SkeletalMeshLocation.SampledPositionWithoutOffset, Vector=<0 0 0>)
    SpriteSize=&Vector2DFromFloat(Value=333)
   )
  ]
  rendererClasses [&NiagaraMeshRendererProperties, &NiagaraSpriteRendererProperties]
  renderers [
   {rendererClass=&NiagaraMeshRendererProperties, rendererIndex=0}
   {rendererClass=&NiagaraMeshRendererProperties, rendererIndex=1}
   {rendererClass=&NiagaraSpriteRendererProperties, rendererIndex=2}
  ]
 }
}
summary {
 systemName=NS_Contrail_PixieParts
 emitters [
  {emitterName=Sprites, rendererClasses [&NiagaraSpriteRendererProperties]}
  {emitterName=Swirls, rendererClasses [&NiagaraMeshRendererProperties]}
  {emitterName=TrailCenter, rendererClasses [&NiagaraMeshRendererProperties, &NiagaraSpriteRendererProperties]}
  {emitterName=Sparks, rendererClasses [&NiagaraSpriteRendererProperties]}
  {emitterName=SymbolPlate, rendererClasses [&NiagaraMeshRendererProperties, &NiagaraLightRendererProperties]}
 ]
 userVariables [
  User.bIsFrontEnd {type=NiagaraBool}
  User.bIsFrontEndPreview {type=NiagaraBool}
  User.ParaGlideLeanAlpha {type=NiagaraFloat, defaultValue=0, description=""}
 ]
}
